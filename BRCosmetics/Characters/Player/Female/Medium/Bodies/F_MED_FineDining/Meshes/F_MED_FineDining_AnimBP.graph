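# Post-process style Anim Blueprint for the F_MED_FineDining female body cosmetic, copying the attached parent's pose into its Control Rig.
# AnimDynamics adds spring-driven motion to the bottle and badge bones, with a weighted multi-bone constraint positioning the badge base between them.
# The event graph caches the player pawn and checks whether its ExtraPart mesh name contains BootsTemplate_Gadget.
# That flag enables Modify Bone nodes that scale both calf twist bones to zero.
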AnimBlueprint /BRCosmetics/Characters/Player/Female/Medium/Bodies/F_MED_FineDining/Meshes/F_MED_FineDining_AnimBP
target_skeleton=&/F_MED_FineDining_Skeleton
graphs [
 AnimGraph {nodes [
   n1 = Control Rig(Source=Copy Pose From Mesh() {bUseAttachedParent=true}) {controlRigAssetReference {blueprintRigClass=&/F_MED_FineDining_CtrlRig.F_MED_FineDining_CtrlRig_C}}
   n2 = Anim Dynamics - Bone: Botle!(ComponentPose=Local To Component(LocalPose=@n1), Alpha=1, GravityOverride:Vector="0,0,-980", ExternalForce="0,0,0") -> (Pose) {
    angularDampingOverride=1
    angularSpringConstant=6000
    bAngularSpring=true
    bOverrideAngularDamping=true
    bUseGravityOverride=true
    gravityScale=0.35
    linearSpringConstant=2000
    simulationSpace=RootRelative
    alphaBoolBlend {blendOption=Linear}
    boundBone {boneName=Botle}
    componentLinearAccScale=<1 1 1>
    componentLinearVelScale=<1 1 1>
    gravityOverride {z=-980}
    physicsBodyDefinitions [{
      boundBone {boneName=Botle}
      boxExtents=<10 10 10>
      constraintSetup {angularLimitsMax=<3 5 0>, angularLimitsMin=<-3 -5 0>, angularTarget=<1 0 0>}
      localJointOffset=<0 0 0.1>
     }]
    relativeSpaceBone {boneName=thigh_r}
   }
   n3 = "Constraint - Base_badge (PPPPPP)"!(
    ComponentPose=@n2
    Alpha=1
    ConstraintWeights_0=0.3506
    ConstraintWeights_1:float=0.1155
    ConstraintWeights_2:float=0.2436
    ConstraintWeights_3:float=0.0231
    ConstraintWeights_4:float=0.0186
    ConstraintWeights_5:float=0.2486
   ) -> (Pose) {
    boneToModify {boneName=Base_badge}
    constraintSetup [
     {targetBone {boneName=Base_badge}}
     {targetBone {boneName=clavicle_r}}
     {targetBone {boneName=spine_04}}
     {targetBone {boneName=upperarm_twist_01_r}}
     {targetBone {boneName=deform_upperarm_fr_r}}
     {targetBone {boneName=deform_pec_r}}
    ]
    constraintWeights [1,1,1,1,1,1]
   }
   n4 = Anim Dynamics - Bone: dyn_badge!(ComponentPose=@n3, Alpha=1, GravityOverride:Vector="0,0,-980", ExternalForce="0,0,0") -> (Pose) {
    angularDampingOverride=1
    angularSpringConstant=6000
    bAngularSpring=true
    bOverrideAngularDamping=true
    bUseGravityOverride=true
    gravityScale=0.35
    linearSpringConstant=2000
    simulationSpace=RootRelative
    alphaBoolBlend {blendOption=Linear}
    boundBone {boneName=dyn_badge}
    componentLinearAccScale=<1 1 1>
    componentLinearVelScale=<1 1 1>
    gravityOverride {z=-980}
    physicsBodyDefinitions [{
      boundBone {boneName=dyn_badge}
      boxExtents=<10 10 10>
      constraintSetup {angularLimitsMax=<0 5 20>, angularLimitsMin=<0 -5 -1>, angularTarget=<1 0 0>}
      localJointOffset=<0 0 0.1>
     }]
    relativeSpaceBone {boneName=thigh_r}
   }
   n5 = Get CustomBootsEquip() -> (CustomBootsEquip:bool)
   n6 = "Transform (Modify) Bone - Bone: calf_twist_01_l"!(ComponentPose=@n4, bAlphaBoolEnabled:bool=@n5, Translation="0,0,0", Rotation="0,0,0", Scale="0,0,0") -> (Pose) {
    alphaInputType=Bool
    scaleMode=BMM_Additive
    translationMode=BMM_Additive
    translationSpace=BCS_BoneSpace
    boneToModify {boneName=calf_twist_01_l}
    scale=<0 0 0>
   }
   n7 = "Transform (Modify) Bone - Bone: calf_twist_01_r"!(ComponentPose=@n6, bAlphaBoolEnabled:bool=@n5, Translation="0,0,0", Rotation="0,0,0", Scale="0,0,0") -> (Pose) {alphaInputType=Bool, scaleMode=BMM_Additive, translationSpace=BCS_BoneSpace, boneToModify {boneName=calf_twist_01_r}, scale=<0 0 0>}
   Output Pose(Result=Component To Local(ComponentPose=@n7)) {layerGroup=None}
  ]}
 EventGraph {nodes [
   Event BlueprintUpdateAnimation()
   > n1 = Sequence()
   n2 = TryGetPawnOwner()
   > Is Valid(InputObject=@n2)
   > n3 = Cast To FortPlayerPawnAthena(Object=@n2)
   > Set As Fort Player Pawn Athena(As Fort Player Pawn Athena:Fort Player Pawn Athena Object Reference=@n3) -> (Output_Get:Fort Player Pawn Athena Object Reference)
   n4 = Get As Fort Player Pawn Athena() -> (As Fort Player Pawn Athena:Fort Player Pawn Athena Object Reference)
   Is Valid(exec=@n1.then_1, InputObject=@n4)
   n5 = GetSkeletalMeshForPartType(self=@n4, PartType=ExtraPart)
   > Set CustomBootsEquip(CustomBootsEquip:bool=Contains(SearchIn=GetDisplayName(Object=@n5), Substring=BootsTemplate_Gadget)) -> (Output_Get:bool)
  ]}
]
variables [
 As Fort Player Pawn Athena {type=FortPlayerPawnAthena}
 CustomBootsEquip {type=boolean}
 Gravity Override {type=Vector}
 Gravity Override Params {type=GravityOverrideParamsStruct}
]
