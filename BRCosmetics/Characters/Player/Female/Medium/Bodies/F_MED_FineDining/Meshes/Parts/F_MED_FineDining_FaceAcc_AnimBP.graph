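# Anim Blueprint for the FineDining female face accessory, simulating its physics, including the ponytail, over a pose copied from the attached parent mesh.
# The front end uses a world-space RigidBody; gameplay uses head-relative RigidBody simulations.
# When customization detects a backbling or cape, gameplay switches to a branch with additive ponytail bone offsets after simulation.
# A deprecated function updates a Gravity Override variable each frame, which no AnimGraph node reads.
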AnimBlueprint /BRCosmetics/Characters/Player/Female/Medium/Bodies/F_MED_FineDining/Meshes/Parts/F_MED_FineDining_FaceAcc_AnimBP
target_skeleton=&/F_MED_FineDining_FaceAcc_Skeleton
graphs [
 AnimGraph {nodes [
   n1 = RigidBody!(ComponentPose=Local To Component(LocalPose=Use cached pose 'SavedPoseFD'()), Alpha, OverrideWorldGravity:Vector="0,0,0", ExternalForce) -> (Pose) {
    bClampLinearTranslationLimitToRefPose=true
    bOverrideWorldGravity=true
    cachedBoundsScale=1
    overridePhysicsAsset=&/F_MED_FineDining_FaceAcc_Physics
    simulationSpace=BaseBoneSpace
    alphaBoolBlend {blendOption=Linear}
    baseBoneRef {boneName=head}
    overrideWorldGravity {z=-980}
    simSpaceSettings {
     maxAngularAcceleration=200
     maxAngularVelocity=3
     maxLinearAcceleration=1400
     maxLinearVelocity=250
     worldAlpha=1
     externalLinearDragV=<0.7 0.4 0.4>
     externalLinearVelocity=<2 0.2 0.2>
    }
   }
   n2 = Component To Local(ComponentPose=@n1)
   n3 = RigidBody(ComponentPose=Local To Component(LocalPose=Use cached pose 'SavedPoseFD'()), OverrideWorldGravity="0,0,0") {
    bClampLinearTranslationLimitToRefPose=true
    bOverrideWorldGravity=true
    cachedBoundsScale=1
    overridePhysicsAsset=&F_MED_FineDining_FaceAcc_Physics
    simulationSpace=BaseBoneSpace
    alphaBoolBlend {blendOption=Linear}
    baseBoneRef {boneName=head}
    overrideWorldGravity {z=-980}
    simSpaceSettings {
     dampingAlpha=2
     maxAngularAcceleration=200
     maxAngularVelocity=3
     maxLinearAcceleration=1400
     maxLinearVelocity=250
     worldAlpha=1
     externalLinearDragV=<0.7 0.4 0.4>
     externalLinearVelocity=<2 0.2 0.2>
    }
   }
   n4 = "Transform (Modify) Bone - Bone: dyn_Hair_ponytail_01"!(ComponentPose=@n3, bAlphaBoolEnabled:bool=True, Translation="0,0,0", Rotation="30,0,0", Scale="1,1,1") -> (Pose) {
    alphaInputType=Bool
    rotationMode=BMM_Additive
    rotationSpace=BCS_BoneSpace
    scaleMode=BMM_Additive
    scaleSpace=BCS_BoneSpace
    translationMode=BMM_Additive
    translationSpace=BCS_BoneSpace
    alphaBoolBlend {blendOption=Linear}
    alphaScaleBiasClamp {bMapRange=true, inRange {max=0, min=-25}, outRange {max=0, min=1}}
    boneToModify {boneName=dyn_Hair_ponytail_01}
    rotation {pitch=30}
   }
   n5 = "Transform (Modify) Bone - Bone: dyn_Hair_ponytail_02"!(ComponentPose=@n4, bAlphaBoolEnabled:bool=True, Translation="0,0,0", Rotation="20,0,0", Scale="0.9,1,1") -> (Pose) {
    alphaInputType=Bool
    rotationMode=BMM_Additive
    rotationSpace=BCS_BoneSpace
    scaleMode=BMM_Additive
    scaleSpace=BCS_BoneSpace
    translationMode=BMM_Additive
    translationSpace=BCS_BoneSpace
    alphaBoolBlend {blendOption=Linear}
    alphaScaleBiasClamp {bMapRange=true, inRange {max=0, min=-25}, outRange {max=0, min=1}}
    boneToModify {boneName=dyn_Hair_ponytail_02}
    rotation {pitch=20}
    scale {x=0.9}
   }
   n6 = "Transform (Modify) Bone - Bone: dyn_Hair_ponytail_03"!(ComponentPose=@n5, bAlphaBoolEnabled:bool=True, Translation="0,0,0", Rotation="8,0,0", Scale="0.85,1,1") -> (Pose) {
    alphaInputType=Bool
    rotationMode=BMM_Additive
    rotationSpace=BCS_BoneSpace
    scaleMode=BMM_Additive
    scaleSpace=BCS_BoneSpace
    translationMode=BMM_Additive
    translationSpace=BCS_BoneSpace
    alphaBoolBlend {blendOption=Linear}
    alphaScaleBiasClamp {bMapRange=true, inRange {max=0, min=-25}, outRange {max=0, min=1}}
    boneToModify {boneName=dyn_Hair_ponytail_03}
    rotation {pitch=8}
    scale {x=0.85}
   }
   n7 = "Transform (Modify) Bone - Bone: dyn_Hair_ponytail_04"!(ComponentPose=@n6, bAlphaBoolEnabled:bool=True, Translation="0,0,0", Rotation="-5,0,0", Scale="0.95,1,1") -> (Pose) {
    alphaInputType=Bool
    rotationMode=BMM_Additive
    rotationSpace=BCS_BoneSpace
    scaleMode=BMM_Additive
    scaleSpace=BCS_BoneSpace
    translationMode=BMM_Additive
    translationSpace=BCS_BoneSpace
    alphaBoolBlend {blendOption=Linear}
    alphaScaleBiasClamp {bMapRange=true, inRange {max=0, min=-25}, outRange {max=0, min=1}}
    boneToModify {boneName=dyn_Hair_ponytail_04}
    rotation {pitch=-5}
    scale {x=0.95}
   }
   n8 = "Transform (Modify) Bone - Bone: dyn_Hair_ponytail_05"!(ComponentPose=@n7, bAlphaBoolEnabled:bool=True, Translation="0,0,0", Rotation="3,0,0", Scale="0.9,0.9,0.9") -> (Pose) {
    alphaInputType=Bool
    rotationMode=BMM_Additive
    rotationSpace=BCS_BoneSpace
    scaleMode=BMM_Additive
    scaleSpace=BCS_BoneSpace
    translationMode=BMM_Additive
    translationSpace=BCS_BoneSpace
    alphaBoolBlend {blendOption=Linear}
    alphaScaleBiasClamp {bMapRange=true, inRange {max=0, min=-25}, outRange {max=0, min=1}}
    boneToModify {boneName=dyn_Hair_ponytail_05}
    rotation {pitch=3}
    scale=<0.9 0.9 0.9>
   }
   n9 = "Transform (Modify) Bone - Bone: dyn_Hair_ponytail_08"!(ComponentPose=@n8, bAlphaBoolEnabled:bool=True, Translation="0,0,0", Rotation="0,0,0", Scale="0.9,0.9,0.9") -> (Pose) {
    alphaInputType=Bool
    rotationMode=BMM_Additive
    rotationSpace=BCS_BoneSpace
    scaleMode=BMM_Additive
    scaleSpace=BCS_BoneSpace
    translationMode=BMM_Additive
    translationSpace=BCS_BoneSpace
    alphaBoolBlend {blendOption=Linear}
    alphaScaleBiasClamp {bMapRange=true, inRange {max=0, min=-25}, outRange {max=0, min=1}}
    boneToModify {boneName=dyn_Hair_ponytail_08}
    scale=<0.9 0.9 0.9>
   }
   n10 = Component To Local(ComponentPose=@n9)
   n11 = Get isWearingCape() -> (isWearingCape:bool)
   n12 = RigidBody(ComponentPose=Local To Component(LocalPose=Use cached pose 'SavedPoseFD'()), OverrideWorldGravity="0,0,-980") {
    bClampLinearTranslationLimitToRefPose=true
    bOverrideWorldGravity=true
    cachedBoundsScale=1
    overridePhysicsAsset=&F_MED_FineDining_FaceAcc_Physics
    simulationSpace=WorldSpace
    alphaBoolBlend {blendOption=Linear}
    baseBoneRef {boneName=head}
    simSpaceSettings {
     maxAngularAcceleration=200
     maxAngularVelocity=3
     maxLinearAcceleration=100
     maxLinearVelocity=70
     velocityScaleZ=0.1
     worldAlpha=1
     externalLinearDragV=<0.7 0.4 0.4>
     externalLinearVelocity=<2 0.2 0.2>
    }
   }
   Output Pose(Result=Blend Poses by bool(
     BlendPose_0=Component To Local(ComponentPose=@n12)
     BlendPose_1=Blend Poses by bool(BlendPose_0=@n10, BlendPose_1=@n2, bActiveValue=@n11) {blendType=Linear}
     bActiveValue=Get bIsInFrontEnd()
    ) {blendType=Linear}) {layerGroup=None}
   Save cached pose 'SavedPoseFD'(Pose=Copy Pose From Mesh() {bUseAttachedParent=true}) {cacheName=SavedPoseFD}
  ]}
 EventGraph {nodes [
   Event BlueprintUpdateAnimation()
   > n1 = Gravity Override State Function DEPRECATED(CharacterPartAnimInstance=Self-Reference(), GameTimeInSeconds=GetGameTimeInSeconds(), GravityOverrideParams=Get Gravity Override Params())
   > Set Gravity Override(Gravity Override=@n1)
   n2 = Event OnCharacterCustomizationLoaded()
   n3 = CheckForBackbling()
   > Set isWearingPack(isWearingPack:bool=@n3) -> (Output_Get:bool)
   > Set isWearingCape(isWearingCape=OR Boolean(A=@n3, B=@n3.bIsBackblingACape))
   > Set BodyMesh(BodyMesh="Get (a copy)"(Array:Array of Skeletal Mesh Component Budgeted Object References=@n2.CustomMeshParts, Dimension 1=1) -> (Output:Skeletal Mesh Component Budgeted Object Reference))
  ]}
]
variables [
 BodyMesh {type=SkeletalMeshComponent}
 Gravity Override {type=Vector}
 Gravity Override Params {type=GravityOverrideParamsStruct}
 isWearingCape {type=boolean}
 isWearingPack {type=boolean}
]
