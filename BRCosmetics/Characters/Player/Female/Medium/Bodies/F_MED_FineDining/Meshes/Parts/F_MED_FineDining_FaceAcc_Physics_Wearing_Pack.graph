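# Physics asset for hair on the female medium FineDining face accessory, in its pack-wearing variant.
# Capsule bodies form a ponytail chain, branching near the tip, and a three-bone bang chain, both with tight per-joint limit values.
# Head, neck, clavicle and upper-spine capsules, including a large offset spine volume, form an angularly locked body frame.
# An extra angularly locked constraint ties ponytail bone 06 directly to spine_05, alongside its chain link.
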PhysicsAsset /BRCosmetics/Characters/Player/Female/Medium/Bodies/F_MED_FineDining/Meshes/Parts/F_MED_FineDining_FaceAcc_Physics_Wearing_Pack
bodies {
 Hair_ponytail_base [{shapeType=Capsule, length=0, radius=0, shapeName=Hair_ponytail_base_capsule, rotation=<rot 0 -90 0>}]
 clavicle_l [{shapeType=Capsule, length=17.8725, radius=5.2946, shapeName=clavicle_l_capsule, center=<4.5698 0 0.0923>, rotation=<rot -0.0003 -90 88.8432>}]
 clavicle_r [{
   shapeType=Capsule
   length=17.8725
   radius=5.2946
   shapeName=clavicle_r_capsule
   center=<-4.5698 0 -0.0923>
   rotation=<rot -0.0003 -90 -91.1568>
  }]
 dyn_Hair_ponytail_01 [{shapeType=Capsule, length=1, radius=1}]
 dyn_Hair_ponytail_02 [{shapeType=Capsule, length=1.574, radius=2.2238}]
 dyn_Hair_ponytail_03 [{shapeType=Capsule, length=1.8011, radius=4.4065, rotation=<rot 65.302 0 0>}]
 dyn_Hair_ponytail_04 [{shapeType=Capsule, length=3.4006, radius=4.0769, rotation=<rot -86.1298 -180 180>}]
 dyn_Hair_ponytail_05 [{shapeType=Capsule, length=3.2394, radius=4.474, rotation=<rot -78.5485 0 0>}]
 dyn_Hair_ponytail_06 [{shapeType=Capsule, length=4.9246, radius=3.9255, rotation=<rot -35.5617 0 0>}]
 dyn_Hair_ponytail_07 [{shapeType=Capsule, length=3.2168, radius=3.6758, rotation=<rot 54.8593 0 0>}]
 dyn_Hair_ponytail_08 [{shapeType=Capsule, length=1, radius=1}]
 dyn_Hair_ponytail_09 [{shapeType=Capsule, length=1, radius=1}]
 dyn_bang_01 [{shapeType=Capsule, length=2.4947, radius=1.1396, center=<1.9863 0 0.2342>, rotation=<rot -83.2758 0.0004 -0.0003>}]
 dyn_bang_02 [{shapeType=Capsule, length=3.8315, radius=1.3656, center=<1.6501 0.0001 0>, rotation=<rot 78.0754 -179.9689 179.9974>}]
 dyn_bang_03 [{shapeType=Capsule, length=2.2767, radius=1.2024, center=<2.4618 0 -0.2359>, rotation=<rot 84.5269 0.0001 0>}]
 head [
  {
   shapeType=Capsule
   length=3.2678
   radius=7.1651
   shapeName=head_capsule
   center=<4.2694 1.3979 0.1401>
   rotation=<rot 3.2034 100.5537 -85.1169>
  }
  {shapeType=Capsule, length=5.5, radius=4.5, center=<0 5.0487 0>, rotation=<rot -89.2809 179.9997 180>}
 ]
 neck_01 [{shapeType=Capsule, length=0.505, radius=0.505, shapeName=neck_01_capsule, rotation=<rot 0 -90 0>}]
 neck_02 [{shapeType=Capsule, length=0.505, radius=0.505, shapeName=neck_02_capsule, rotation=<rot 0 -90 0>}]
 spine_04 [{shapeType=Capsule, length=0.505, radius=0.505, shapeName=spine_04_capsule, rotation=<rot 0 -90 0>}]
 spine_05 [
  {shapeType=Capsule, length=21.9476, radius=10.2964, shapeName=spine_05_capsule, rotation=<rot 0 -90 0>}
  {shapeType=Capsule, length=15, radius=15, shapeName=spine_05_capsule, center=<7.421 12.193 0>}
 ]
}
constraints [
 {
  bone1Name=Hair_ponytail_base
  bone2Name=head
  swing1LimitDegrees=45
  swing1Motion=Locked
  swing2LimitDegrees=45
  swing2Motion=Locked
  twistLimitDegrees=45
  twistMotion=Locked
 }
 {bone1Name=dyn_Hair_ponytail_01, bone2Name=Hair_ponytail_base, swing1LimitDegrees=20, swing2LimitDegrees=16, twistLimitDegrees=10}
 {bone1Name=dyn_Hair_ponytail_02, bone2Name=dyn_Hair_ponytail_01, swing1LimitDegrees=10.32, swing2LimitDegrees=10.76, twistLimitDegrees=1}
 {bone1Name=dyn_Hair_ponytail_03, bone2Name=dyn_Hair_ponytail_02, swing1LimitDegrees=15.76, swing2LimitDegrees=15.76, twistLimitDegrees=1}
 {bone1Name=dyn_Hair_ponytail_04, bone2Name=dyn_Hair_ponytail_03, swing1LimitDegrees=20, swing2LimitDegrees=25, twistLimitDegrees=1}
 {bone1Name=dyn_Hair_ponytail_05, bone2Name=dyn_Hair_ponytail_04, swing1LimitDegrees=17.2, swing2LimitDegrees=15, twistLimitDegrees=1}
 {bone1Name=dyn_Hair_ponytail_06, bone2Name=dyn_Hair_ponytail_05, swing1LimitDegrees=4, swing2LimitDegrees=5, twistLimitDegrees=1}
 {bone1Name=dyn_Hair_ponytail_07, bone2Name=dyn_Hair_ponytail_06, swing1LimitDegrees=10, swing2LimitDegrees=5, twistLimitDegrees=1}
 {bone1Name=dyn_Hair_ponytail_08, bone2Name=dyn_Hair_ponytail_05, swing1LimitDegrees=10, swing2LimitDegrees=5, twistLimitDegrees=1}
 {bone1Name=dyn_Hair_ponytail_09, bone2Name=dyn_Hair_ponytail_04, swing1LimitDegrees=10, swing2LimitDegrees=5, twistLimitDegrees=1}
 {bone1Name=dyn_bang_01, bone2Name=head, swing1LimitDegrees=4, swing2LimitDegrees=4, twistLimitDegrees=1}
 {bone1Name=dyn_bang_02, bone2Name=dyn_bang_01, swing1LimitDegrees=5, swing2LimitDegrees=5, twistLimitDegrees=1}
 {bone1Name=dyn_bang_03, bone2Name=dyn_bang_02, swing1LimitDegrees=10, swing2LimitDegrees=10, twistLimitDegrees=4}
 {
  bone1Name=head
  bone2Name=neck_02
  swing1LimitDegrees=45
  swing1Motion=Locked
  swing2LimitDegrees=45
  swing2Motion=Locked
  twistLimitDegrees=45
  twistMotion=Locked
 }
 {
  bone1Name=neck_02
  bone2Name=neck_01
  swing1LimitDegrees=45
  swing1Motion=Locked
  swing2LimitDegrees=45
  swing2Motion=Locked
  twistLimitDegrees=45
  twistMotion=Locked
 }
 {
  bone1Name=neck_01
  bone2Name=spine_05
  swing1LimitDegrees=45
  swing1Motion=Locked
  swing2LimitDegrees=45
  swing2Motion=Locked
  twistLimitDegrees=45
  twistMotion=Locked
 }
 {
  bone1Name=spine_05
  bone2Name=spine_04
  swing1LimitDegrees=45
  swing1Motion=Locked
  swing2LimitDegrees=45
  swing2Motion=Locked
  twistLimitDegrees=45
  twistMotion=Locked
 }
 {
  bone1Name=clavicle_l
  bone2Name=spine_05
  swing1LimitDegrees=45
  swing1Motion=Locked
  swing2LimitDegrees=45
  swing2Motion=Locked
  twistLimitDegrees=45
  twistMotion=Locked
 }
 {
  bone1Name=clavicle_r
  bone2Name=spine_05
  swing1LimitDegrees=45
  swing1Motion=Locked
  swing2LimitDegrees=45
  swing2Motion=Locked
  twistLimitDegrees=45
  twistMotion=Locked
 }
 {
  bone1Name=dyn_Hair_ponytail_06
  bone2Name=spine_05
  swing1LimitDegrees=45
  swing1Motion=Locked
  swing2LimitDegrees=45
  swing2Motion=Locked
  twistLimitDegrees=45
  twistMotion=Locked
 }
]
